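; Targeting selection trace task for creative melee weapons that supplies the swept trace radius for the cannonball impact.
; It casts the targeting request's source context to a player pawn and its current weapon to Creative_Melee_Parent.
; That weapon returns its cannonball impact radius through CreativeMeleeWeaponInterface.
; If either cast fails, a hardcoded fallback radius is used.
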(fn GetSweptTraceRadius
  (params (TargetingHandle :type Targeting%20Request%20Handle%20Structure%20%28by%20ref%29))
  (bind _asfort_player_pawn (Utilities|Casting|CastToFortPlayerPawn (Utilities|Struct|BreakTargetingSourceContext (TargetingSystem|TargetingTypes|GetTargetingSourceContext "/Script/TargetingSystem.Default__TargetingSubsystem" TargetingHandle)))
    (:then
      (bind _ascreative_melee_parent (Utilities|Casting|CastToCreative_Melee_Parent (Variables|/Script/FortniteGame.FortPawn|GetCurrentWeapon _asfort_player_pawn))
        (:then
          (bind _impactradius (CreativeMeleeWeaponInterface|GetCannonballImpactRadius _ascreative_melee_parent :!class /CRD_Melee/Weapons/Creative_Melee_Parent.Creative_Melee_Parent_C))
          (return _impactradius))
        (:CastFailed
          _)))
    (:CastFailed
      _))
  (return 256.0))
