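; Target selection trace task that finds aim assist targets for Mantis melee attacks along the player's movement input rather than snapping against it.
; A shared helper casts the targeting source context to a Fort player pawn and reads its movement component's last input vector.
; That vector sets the trace direction and pushes the start forward from the pawn by the swept trace radius override, keeping targets behind the player out.
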(fn GetPlayerIntendedDirection
  (params (TargetingHandle :type Targeting%20Request%20Handle%20Structure))
  ;[0 0.7015 1 0.25] Since the Mantis attacks are targeted by input direction, we need to use this to determine what we want to lock onto (to avoid trying to move in one direction and have the pawn snap in the other direction)
  (bind _asfort_player_pawn (Utilities|Casting|CastToFortPlayerPawn (Utilities|Struct|BreakTargetingSourceContext (TargetingSystem|TargetingTypes|GetTargetingSourceContext "/Script/TargetingSystem.Default__TargetingSubsystem" TargetingHandle))))
  (return _asfort_player_pawn (Pawn|Components|PawnMovement|GetLastInputVector (Variables|/Script/Engine.Character|GetCharacterMovement _asfort_player_pawn))))

(fn GetSourceLocation
  (params (TargetingHandle :type Targeting%20Request%20Handle%20Structure%20%28by%20ref%29))
  (bind (_source_pawn _returnvalue) (CallFunction|GetPlayerIntendedDirection TargetingHandle))
  ;[0 0.7015 1 0.25] Source Location is in-front of the player at the distance of the Trace Radius (to stop targeting behind the player)
  (return (+ (Transformation|GetActorLocation _source_pawn) (* _returnvalue (Math|Float|MakeLiteralFloat (Variables|Default|GetSweptTraceRadiusOverride))))))

(fn GetSweptTraceRadius
  (params (TargetingHandle :type Targeting%20Request%20Handle%20Structure%20%28by%20ref%29))
  (return (Variables|Default|GetSweptTraceRadiusOverride)))

(fn GetTraceDirection
  (params (TargetingHandle :type Targeting%20Request%20Handle%20Structure%20%28by%20ref%29))
  (bind (_source_pawn _returnvalue) (CallFunction|GetPlayerIntendedDirection TargetingHandle))
  (return _returnvalue))
